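; Blueprint for the SeaMist lighthouse Fresnel lens prop, handling cleanup of its light beam effect when the actor is destroyed.
; On destruction it destroys the current FX actor of its Light Beam PersistentFX component, then the component itself.
; It also destroys every actor in its ActorsToDetroyAlongWith array, noted as volumes tied to the light, before calling the parent implementation.
; BeginPlay, overlap and Tick events are empty, and the construction script only calls its parent.
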(event EventBeginPlay)

(event Collision|EventActorBeginOverlap
  (params (OtherActor :type Actor%20Object%20Reference)))

(event EventTick
  (params (DeltaSeconds :type Float%20%28single-precision%29)))

(event EventDestroyed
  (bind _persistentfxcomponent_light_beam (Variables|Default|GetPersistentFXComponent%20Light%20Beam))
  (Actor|DestroyActor (Variables|/Game/Blueprints/Components/PersistentFXComponent/PersistentFXComponent.PersistentFXComponent_C|GetcurrentFxActor _persistentfxcomponent_light_beam))
  (Components|DestroyComponent _persistentfxcomponent_light_beam)
  (for _array_element (Variables|Default|GetActorsToDetroyAlongWith)
    (Actor|DestroyActor _array_element) ; Adding volumes as children to be destroyed when light is destroyed
  )
  (|CallParentFunction))

(fn ConstructionScript
  (|CallParentFunction))
